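; Prototype weapon Blueprint for the HighTower HoneyDew dual-wield melee fists.
; On equip it caches the player mesh, binds swing callbacks through a player interface, and attaches both fist meshes and idle FX.
; Swing events and a Swing4 GameplayCue reposition one detached Niagara component with debug prints, while a Smash cue spawns a ground-impact emitter, all cosmetic-gated.
; Impacts add sound, a randomly rotated fading decal and camera shake.
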(event Custom|Swing%20Left
  (bind (_returnvalue_location _returnvalue_rotation) (Math|Transform|BreakTransform (Transformation|GetWorldTransform (Variables|Default|GetPlayerMesh))))
  (bind _swingfx_niagara (Variables|Default|GetSwingFX%20Niagara))
  ;[1 1 1] SwingFX
  (Networking|CanExecuteCosmeticEvents
    (:True
      (Development|PrintString "Swing Left" true true "(R=1.000000,G=0.736136,B=0.000000,A=1.000000)")
      (Components|Activation|Activate _swingfx_niagara true)
      (Transformation|SetWorldTransform _swingfx_niagara (Math|Transform|MakeTransform _returnvalue_location _returnvalue_rotation) false false)
      (Transformation|AddLocalRotation _swingfx_niagara "175.000000, 0,0.000000" false false)
      (Transformation|AddWorldOffset _swingfx_niagara "0, 0,100.000000" false false))))

(event Custom|Swing%20Left%20End)

(event Custom|Swing%20Right%20End)

(event Custom|Swing%20Right
  (bind (_returnvalue_location _returnvalue_rotation) (Math|Transform|BreakTransform (Transformation|GetWorldTransform (Variables|Default|GetPlayerMesh))))
  (bind _swingfx_niagara (Variables|Default|GetSwingFX%20Niagara))
  ;[1 1 1] SwingFX
  (Networking|CanExecuteCosmeticEvents
    (:True
      (Development|PrintString "Swing Right" true true "(R=1.000000,G=0.736136,B=0.000000,A=1.000000)")
      (Components|Activation|Activate _swingfx_niagara true)
      (Transformation|SetWorldTransform _swingfx_niagara (Math|Transform|MakeTransform _returnvalue_location _returnvalue_rotation) false false)
      (Transformation|AddLocalRotation _swingfx_niagara "-15.000000,15.000000,0.000000" false false)
      (Transformation|AddWorldOffset _swingfx_niagara "0, 0,140.000000" false false))))

(event Custom|Swing%20Left%202
  (bind (_returnvalue_location _returnvalue_rotation) (Math|Transform|BreakTransform (Transformation|GetWorldTransform (Variables|Default|GetPlayerMesh))))
  (bind _swingfx_niagara (Variables|Default|GetSwingFX%20Niagara))
  ;[1 1 1] SwingFX
  (Networking|CanExecuteCosmeticEvents
    (:True
      (Development|PrintString "Swing Left 2" true true "(R=1.000000,G=0.736136,B=0.000000,A=1.000000)")
      (Components|Activation|Activate _swingfx_niagara true)
      (Transformation|SetWorldTransform _swingfx_niagara (Math|Transform|MakeTransform _returnvalue_location _returnvalue_rotation) false false)
      (Transformation|AddLocalRotation _swingfx_niagara "220.000000,-12.000000,0.000000" false false)
      (Transformation|AddWorldOffset _swingfx_niagara "0, 0,165.000000" false false))))

(event Custom|Swing%20Right%202)

(event GameplayCue|GameplayCue.HighTower.HoneyDew.Fist.Swing4
  (params (EventType :type EGameplayCueEvent%20Enum :default "OnActive") (Parameters :type Gameplay%20Cue%20Parameters%20Structure%20%28by%20ref%29))
  (bind (_returnvalue_location _returnvalue_rotation) (Math|Transform|BreakTransform (Transformation|GetWorldTransform (Variables|Default|GetPlayerMesh))))
  (bind _swingfx_niagara (Variables|Default|GetSwingFX%20Niagara))
  ;[1 1 1] SwingFX
  (Networking|CanExecuteCosmeticEvents
    (:True
      (Development|PrintString "Swing Right 2" true true "(R=1.000000,G=0.736136,B=0.000000,A=1.000000)")
      (Components|Activation|Activate _swingfx_niagara true)
      (Transformation|SetWorldTransform _swingfx_niagara (Math|Transform|MakeTransform _returnvalue_location _returnvalue_rotation) false false)
      (Transformation|AddLocalRotation _swingfx_niagara "90.000000,0.000000,0.000000" false false)
      (Transformation|AddWorldOffset _swingfx_niagara "0, 0,100.000000" false false))))

(event GameplayCue|GameplayCue.HighTower.HoneyDew.Fist.Smash
  (params (EventType :type EGameplayCueEvent%20Enum :default "OnActive") (Parameters :type Gameplay%20Cue%20Parameters%20Structure%20%28by%20ref%29))
  (bind (_returnvalue_location _returnvalue_rotation) (Math|Transform|BreakTransform (Transformation|GetWorldTransform (Variables|Default|GetPlayerMesh))))
  ;[1 1 1] SwingFX
  (Networking|CanExecuteCosmeticEvents
    (:True
      (bind _returnvalue (Effects|SpawnEmitteratLocation "/HighTower/Items/HoneyDew/Fist/FX/P_HighTower_HoneyDew_MeleeFist_ImpactGround.P_HighTower_HoneyDew_MeleeFist_ImpactGround" _returnvalue_location _returnvalue_rotation))
      (Transformation|AddLocalOffset _returnvalue "0,190.000000,10.000000" false false))))

(event FortWeapon|EventOnEquippedAndAttached
  ;[1 1 1] On Equip
  (|CallParentFunction)
  (bind _asfort_pawn (Utilities|Casting|CastToFortPawn (Game|GetInstigator)))
  (Variables|Default|SetPlayerMesh (Variables|/Script/Engine.Character|GetMesh _asfort_pawn))
  (Utilities|IsValid (Variables|/Script/Engine.Character|GetMesh _asfort_pawn)
    (:Is%20Valid
      _)
    (:Is%20Not%20Valid
      (Utilities|FlowControl|Delay 0.1)))
  (CallFunction|BindDualMeleeSwingEvents)
  (Transformation|AttachComponentToComponent (Variables|Visual|GetWeaponMesh) (Variables|/Script/Engine.Character|GetMesh _asfort_pawn) (Variables|Default|GetMainhandSocketName) "SnapToTarget" "SnapToTarget" "SnapToTarget" true)
  (Transformation|AttachComponentToComponent (Variables|Default|GetWeaponMesh_Offhand) (Variables|/Script/Engine.Character|GetMesh _asfort_pawn) (Variables|Default|GetOffhandSocketName) "SnapToTarget" "SnapToTarget" "SnapToTarget" true)
  (Transformation|SetRelativeLocationAndRotation (Variables|Visual|GetWeaponMesh) (Variables|Default|GetWeaponRelativeOffset_Main) (Variables|Default|GetWeaponRelativeRotation_Main) false true)
  (Transformation|SetRelativeLocationAndRotation (Variables|Default|GetWeaponMesh_Offhand) (Variables|Default|GetWeaponRelativeOffset_Offhand) (Variables|Default|GetWeaponRelativeRotation_Offhand) false true)
  ;[1 1 1] Idle FX
  (Transformation|AttachComponentToComponent (Variables|Default|GetIdleFX%20Niagara) (Variables|/Script/Engine.Character|GetMesh _asfort_pawn) "root" "SnapToTarget" "SnapToTarget" "SnapToTarget" true))

(event EventOnWeaponVisibilityChanged
  (params (bVisible :type Boolean) (bSetForLocalControllerOnly :type Boolean))
  ;[1 1 1] Idle FX
  (|CallParentFunction false false)
  (Components|Activation|Activate (Variables|Default|GetIdleFX%20Niagara) true))

(event FortWeapon|EventOnLocalUnEquip
  ;[1 1 1] On Unequip / Detached
  (|CallParentFunction)
  (Transformation|DetachFromComponent (Variables|Default|GetWeaponMesh_Offhand) "KeepRelative" "KeepRelative" "KeepRelative" true))

(event FortWeapon|EventOnPlayImpactFX
  (params (HitResult :type Hit%20Result%20Structure%20%28by%20ref%29) (ImpactPhysicalSurface :type EPhysicalSurface%20Enum :default "SurfaceType_Default") (SpawnedPSC :type FXSystem%20Component%20Object%20Reference))
  (bind _returnvalue_roll (Math|Rotator|BreakRotator (Math|Random|RandomRotator true)))
  (bind (_bblockinghit _binitialoverlap _time _distance _location _impactpoint _normal _impactnormal) (Collision|BreakHitResult HitResult))
  ;[1 1 1] Decal on hit
  (|CallParentFunction HitResult ImpactPhysicalSurface SpawnedPSC)
  (Audio|PlaySoundatLocation (Variables|Default|GetImpact%20SFX) _impactpoint)
  (bind _returnvalue (Rendering|Decal|SpawnDecalatLocation (Variables|Default|GetImpactDecal) (Variables|Default|GetDecalSize) _impactpoint (Math|Conversions|RotationFromXVector _impactnormal) 2.0))
  (Transformation|AddLocalRotation _returnvalue (Math|Rotator|MakeRotator _returnvalue_roll) false false)
  (Rendering|Components|Decal|SetFadeOut _returnvalue 1.0 2.0 true)
  (Camera|PlayWorldCameraShake "/HighTower/Items/HoneyDew/Fist/FX/BP_Camera_Shake_HighTower_HoneyDew_FistPunch.BP_Camera_Shake_HighTower_HoneyDew_FistPunch_C" (Transformation|GetWorldLocation (Variables|Default|GetPlayerMesh)) 3e+02 5e+02))

(fn BindDualMeleeSwingEvents
  (bind _asmelee_swing_and_anim_trail_events_player_interface (Utilities|Casting|CastToMeleeSwingAndAnimTrailEventsPlayerInterface (Game|GetInstigator))
    (:then
      (MeleeSwingAndAnimTrailEventsPlayerInterface|BindtoSwingFX _asmelee_swing_and_anim_trail_events_player_interface (EventDispatchers|CreateEvent :!event Swing%20Left) (EventDispatchers|CreateEvent :!event Swing%20Left%20End) (EventDispatchers|CreateEvent :!event Swing%20Right) (EventDispatchers|CreateEvent :!event Swing%20Right%20End) :!class /Script/FortniteGame.MeleeSwingAndAnimTrailEventsPlayerInterface)
      (MeleeSwingAndAnimTrailEventsPlayerInterface|BindtoSwingFX2 _asmelee_swing_and_anim_trail_events_player_interface (EventDispatchers|CreateEvent :!event Swing%20Left%202) (EventDispatchers|CreateEvent :!event Swing%20Right%202) :!class /Script/FortniteGame.MeleeSwingAndAnimTrailEventsPlayerInterface))
    (:CastFailed
      (Development|PrintString (Utilities|String|Append "Cannot bind events.  Cast Failed!  Instigator of dual-wield melee weapon is not a MeleeSwingAndAnimTrailEventsPlayerInterface, so it will not dual-wield properly. Instigator: " (Utilities|String|ToString(Object) (Game|GetInstigator)))) ; Instigator class is not supported!
    )))

(fn OnInitWeaponCosmetics)

(fn ConstructionScript
  (|CallParentFunction)
  (Transformation|DetachFromComponent (Variables|Default|GetSwingFX%20Niagara) "KeepWorld" "KeepWorld" "KeepWorld" true))
